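# Physics asset for the one-handed Teriyaki Fish Elf pickaxe cosmetic, simulating a fish dangling from the weapon.
# It has a large capsule on the handle bone and small capsules on the ring, loop and fish bones.
# Constraints chain handle to ring to loop to fish, each with near-locked twist.
# Swing limits are asymmetric, and the dominant axis alternates along the chain so the links hinge like a charm.
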PhysicsAsset /BRCosmetics/Weapons/FORT_Melee/Pickaxe_Teriyaki_Fish_Elf_1H/Meshes/Teriyaki_Fish_Elf_1h_Axe_Physics
bodies {
 Melee_TwoHanded_Handle [{shapeType=Capsule, length=30.8845, radius=17.6223, center=<-0.0843 -1.2174 16.6351>, rotation=<rot -8.0657 -90 -0.0014>}]
 dyn_fish [{shapeType=Capsule, length=2.5, radius=2.005, center=<3.6646 0.0001 0>, rotation=<rot 90 0 0>}]
 dyn_loop [{shapeType=Capsule, length=2.5, radius=2.005, center=<3.9307 0 -0.0143>, rotation=<rot 0 -89.9997 90.2081>}]
 dyn_ring [{shapeType=Capsule, length=2.5, radius=2.005, center=<3.4406 0 -0.0125>, rotation=<rot 0 -89.9997 90.2081>}]
}
constraints [
 {bone1Name=dyn_fish, bone2Name=dyn_loop, swing1LimitDegrees=15, swing2LimitDegrees=90, twistLimitDegrees=1}
 {bone1Name=dyn_loop, bone2Name=dyn_ring, swing1LimitDegrees=65, swing2LimitDegrees=2, twistLimitDegrees=1}
 {bone1Name=dyn_ring, bone2Name=Melee_TwoHanded_Handle, swing1LimitDegrees=1, swing2LimitDegrees=75, twistLimitDegrees=1}
]
